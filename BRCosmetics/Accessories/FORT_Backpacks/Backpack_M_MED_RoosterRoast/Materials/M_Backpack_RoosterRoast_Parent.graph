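# Parent material for the RoosterRoast backpack that adds animated scanlines to what its graph treats as a screen area.
# Between the 2023 character PreFX and PostFX functions, it blends scanline-multiplied versions of Base Color and Emissive back into the material attributes.
# The screen mask is derived procedurally from a pre-skinned local Z threshold and non-zero emissive.
# Scanlines use swizzled, tiled UVs and fade with camera depth.
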Material /BRCosmetics/Accessories/FORT_Backpacks/Backpack_M_MED_RoosterRoast/Materials/M_Backpack_RoosterRoast_Parent
graph {
 properties {
  bHasPixelAnimation=true
  bUseMaterialAttributes=true
  bUsedWithMeshDeformer=true
  bUsedWithMorphTargets=true
  bUsedWithSkeletalMesh=true
  pixelDepthOffsetMode=PDOM_Legacy
  previewMesh=/BRCosmetics/Accessories/FORT_Backpacks/Backpack_M_MED_RoosterRoast/Meshes/M_MED_RoosterRoast_Pack
  shadingModel=MSM_FromMaterialExpression
  usedShadingModels=Default Lit | Subsurface | Clear Coat | Subsurface Profile | Cloth
 }
 nodes [
  n1 = MaterialFunctionCall() {materialFunction=&MF_CharacterParent_2023_PreFX}
  MaterialExpressionGetMaterialAttributes_0 = GetMaterialAttributes(None=@n1.Result) {attributeGetTypes [#1, #2]}
  n2 = NamedRerouteUsage()
  n3 = Multiply(A=@MaterialExpressionGetMaterialAttributes_0.Emissive Color, B=@n2)
  n4 = NamedRerouteUsage()
  n5 = Multiply(A=@MaterialExpressionGetMaterialAttributes_0.Base Color, B=@n2)
  MaterialFunctionCall(
   Material=SetMaterialAttributes!(
    MaterialAttributes=@MaterialExpressionGetMaterialAttributes_0.MaterialAttributes
    Emissive Color=LinearInterpolate(A=@MaterialExpressionGetMaterialAttributes_0.Emissive Color, B=@n3, Alpha=@n4)
    Base Color=LinearInterpolate(A=@MaterialExpressionGetMaterialAttributes_0.Base Color, B=@n5, Alpha=@n4)
   ) {attributeSetTypes [#1, #2]}
   M=@n1.M
   Pre Skinned Local Position=@n1.M
   Distance Blend=@n1.M
   Fresnel=@n1.M
  ) {materialFunction=&MF_CharacterParent_2023_PostFX}
  n6 = ComponentMask(None=@n1.Pre skinned Local Position) {b=true}
  n7 = Convert(None=@MaterialExpressionGetMaterialAttributes_0.Emissive Color) {
   nodeName=Break Float3
   convertInputs [{
     type=Vector3
     expressionInput {
      expression=&self:MaterialExpressionGetMaterialAttributes_0
      inputName=None
      mask=0
      maskA=0
      maskB=0
      maskG=0
      maskR=0
      outputIndex=1
     }
    }]
   convertOutputs [{type=Scalar}, {type=Scalar}, {type=Scalar}]
  }
  n8 = Max(A=Max(A=@n7, B=@n7), B=@n7)
  NamedRerouteDeclaration(None=Multiply(A=Step(X=@n6) {constY=-2}, B=Ceil(None=@n8))) {name=Screen Mask, nodeColor=<rgba 0.2422 0 1 1>}
  n9 = NamedRerouteUsage()
  n10 = MaterialFunctionCall!(
   Number of Scanlines
   Scanline Speed
   "Scanlines Thickness (Bias)"
   Scanlines Dark Value
   Scanlines Light Value
   Block Opacity
   UVs=MaterialFunctionCall(XY=TextureCoordinate() {uTiling=1.5}) {materialFunction=&Swizzle}.YX
   Time
   Block Repeat
  ) {materialFunction=/Game/Effects/Fort_Effects/Effects/Atlas/Materials/MF_Atlas_Scanlines_01}
  n11 = ScalarParameter() {defaultValue=300, group=RoosterRoast, parameterName=Scanlines_FadeLength}
  n12 = ScalarParameter() {defaultValue=200, group=RoosterRoast, parameterName=Scanlines_FadeOffset}
  NamedRerouteDeclaration(None=LinearInterpolate(
    A=Multiply(A=@n9, B=@n10.Result)
    Alpha=MaterialFunctionCall(Fade Length=@n11, Fade Offset=@n12) {materialFunction=&CameraDepthFade}.Result
   )) {name=Masked Scanlines, nodeColor=<rgba 0 1 0.3203 1>}
 ]
}
