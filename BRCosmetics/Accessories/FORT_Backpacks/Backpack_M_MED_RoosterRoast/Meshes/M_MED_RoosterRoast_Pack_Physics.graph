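# Physics asset for the M_MED_RoosterRoast backpack cosmetic, giving its jar, coins and bill capsule bodies for physical motion.
# Constraints chain the jar's base and dynamic bones off body_cr, then branch from the dynamic jar into three coin chains and a two-segment bill chain.
# Links into each base bone and the dynamic jar rotate freely, while dynamic coin and bill links lock twist and have tight swing limits.
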PhysicsAsset /BRCosmetics/Accessories/FORT_Backpacks/Backpack_M_MED_RoosterRoast/Meshes/M_MED_RoosterRoast_Pack_Physics
bodies {
 base_bill_01_rt [{shapeType=Capsule, length=0.505, radius=0.505, shapeName=base_bill_01_rt_capsule, rotation=<rot 0 -90 0>}]
 base_coin_01_rt [{shapeType=Capsule, length=0.505, radius=0.505, shapeName=base_coin_01_rt_capsule, rotation=<rot 0 -90 0>}]
 base_coin_02_rt [{shapeType=Capsule, length=0.505, radius=0.505, shapeName=base_coin_02_rt_capsule, rotation=<rot 0 -90 0>}]
 base_coin_03_rt [{shapeType=Capsule, length=0.505, radius=0.505, shapeName=base_coin_03_rt_capsule, rotation=<rot 0 -90 0>}]
 base_jar_rt [{shapeType=Capsule, length=0.505, radius=0.505, shapeName=base_jar_rt_capsule, rotation=<rot 0 -90 0>}]
 body_cr [{shapeType=Capsule, length=14.7026, radius=13.2278, center=<-1.7466 -7.9691 0.4475>, rotation=<rot 0 71.2659 -87.4855>}]
 dyn_bill_01_rt [{shapeType=Capsule, length=2, radius=1, shapeName=dyn_bill_01_rt_capsule, center=<5 0 0>, rotation=<rot 90 0 0>}]
 dyn_bill_02_rt [{shapeType=Capsule, length=2, radius=1, shapeName=dyn_bill_02_rt_capsule, center=<5 0 0>, rotation=<rot 90 0 0>}]
 dyn_coin_01_rt [{shapeType=Capsule, length=2, radius=1, shapeName=dyn_coin_01_rt_capsule, center=<5 0 0>, rotation=<rot 90 0 0>}]
 dyn_coin_02_rt [{shapeType=Capsule, length=2, radius=1, shapeName=dyn_coin_02_rt_capsule, center=<5 0 0>, rotation=<rot 90 0 0>}]
 dyn_coin_03_rt [{shapeType=Capsule, length=2, radius=1, shapeName=dyn_coin_03_rt_capsule, center=<5 0 0>, rotation=<rot 90 0 0>}]
 dyn_jar_rt [{
   shapeType=Capsule
   length=4.1445
   radius=3.4595
   shapeName=dyn_jar_rt_capsule
   center=<3.7651 0.0655 1.4482>
   rotation=<rot -3.2142 -90 0.1117>
  }]
}
constraints [
 {
  bone1Name=base_jar_rt
  bone2Name=body_cr
  swing1LimitDegrees=45
  swing1Motion=Free
  swing2LimitDegrees=45
  swing2Motion=Free
  twistLimitDegrees=45
  twistMotion=Free
 }
 {
  bone1Name=dyn_jar_rt
  bone2Name=base_jar_rt
  swing1LimitDegrees=45
  swing1Motion=Free
  swing2LimitDegrees=45
  swing2Motion=Free
  twistLimitDegrees=45
  twistMotion=Free
 }
 {
  bone1Name=base_bill_01_rt
  bone2Name=dyn_jar_rt
  swing1LimitDegrees=45
  swing1Motion=Free
  swing2LimitDegrees=45
  swing2Motion=Free
  twistLimitDegrees=45
  twistMotion=Free
 }
 {bone1Name=dyn_bill_01_rt, bone2Name=base_bill_01_rt, swing1LimitDegrees=0, swing2LimitDegrees=10, twistLimitDegrees=0, twistMotion=Locked}
 {bone1Name=dyn_bill_02_rt, bone2Name=dyn_bill_01_rt, swing1LimitDegrees=0, swing2LimitDegrees=10, twistLimitDegrees=0, twistMotion=Locked}
 {
  bone1Name=base_coin_01_rt
  bone2Name=dyn_jar_rt
  swing1LimitDegrees=45
  swing1Motion=Free
  swing2LimitDegrees=45
  swing2Motion=Free
  twistLimitDegrees=45
  twistMotion=Free
 }
 {bone1Name=dyn_coin_01_rt, bone2Name=base_coin_01_rt, swing1LimitDegrees=10, swing2LimitDegrees=10, twistLimitDegrees=0, twistMotion=Locked}
 {
  bone1Name=base_coin_02_rt
  bone2Name=dyn_jar_rt
  swing1LimitDegrees=45
  swing1Motion=Free
  swing2LimitDegrees=45
  swing2Motion=Free
  twistLimitDegrees=45
  twistMotion=Free
 }
 {bone1Name=dyn_coin_02_rt, bone2Name=base_coin_02_rt, swing1LimitDegrees=10, swing2LimitDegrees=10, twistLimitDegrees=0, twistMotion=Locked}
 {
  bone1Name=base_coin_03_rt
  bone2Name=dyn_jar_rt
  swing1LimitDegrees=45
  swing1Motion=Free
  swing2LimitDegrees=45
  swing2Motion=Free
  twistLimitDegrees=45
  twistMotion=Free
 }
 {bone1Name=dyn_coin_03_rt, bone2Name=base_coin_03_rt, swing1LimitDegrees=10, swing2LimitDegrees=10, twistLimitDegrees=0, twistMotion=Locked}
]
